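# Control Rig for the Mastermind head that adjusts its dynamic tube bones as the character moves.
# Lower tube masters are parent-constrained to clavicles and spine_05.
# Two spherical pose readers on head rotation then weight corrective relative transforms, both on the left master; the second is relative to the right master and disabled when isinFrontEnd is true.
# Upper-tube aim, debug-draw and transform-constraint nodes are disconnected from execution.
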ControlRigBlueprint /BRCosmetics/Characters/Player/Male/Medium/Heads/M_MED_Mastermind_Head/Meshes/M_MED_Mastermind_CtrlRig
graphs [
 RigVMFunctionLibrary {links [], nodes []}
 RigVMModel {nodes [
   /Script/ControlRig.RigUnit_AimBone(
    Bone=dyn_up_tube_r_MSTR
    DebugSettings=(bEnabled=false,Scale=10,WorldOffset=(Rotation=(X=0,Y=0,Z=0,W=1),Translation=(X=0,Y=0,Z=0),Scale3D=(X=1,Y=1,Z=1)))
    ExecutePin=()
    Primary=(Weight=1,Axis=(X=1,Y=0,Z=0),Target=(X=1,Y=0,Z=0),Kind=Location,Space="dyn_up_tube_r_TGT")
    Secondary=(Weight=1,Axis=(X=0,Y=1,Z=0),Target=(X=0,Y=1,Z=0),Kind=Direction,Space="dyn_up_tube_r_TGT")
    Weight=1
    bPropagateToChildren=True
   ) {is=struct RigVMUnitNode}
   > &RigUnit_AimBone(
    Bone=dyn_up_tube_l_MSTR
    DebugSettings=(bEnabled=False,Scale=10,WorldOffset=(Rotation=(X=0,Y=0,Z=0,W=1),Translation=(X=0,Y=0,Z=0),Scale3D=(X=1,Y=1,Z=1)))
    Primary=(Weight=1,Axis=(X=1,Y=0,Z=0),Target=(X=1,Y=0,Z=0),Kind=Location,Space="dyn_up_tube_l_TGT")
    Secondary=(Weight=1,Axis=(X=0,Y=1,Z=0),Target=(X=0,Y=1,Z=0),Kind=Direction,Space="dyn_up_tube_l_TGT")
    Weight=1
    bPropagateToChildren=True
   )
   n1 = &RigUnit_BeginExecution()
   &RigUnit_ParentConstraint(
    AdvancedSettings=(InterpolationType=Average,RotationOrderForFilter=XZY)
    Child=(Type=Bone,Name="dyn_lw_tube_l_MSTR")
    Parents=((Item=(Type=Bone,Name="clavicle_l"),Weight=0.5),(Item=(Type=Bone,Name="spine_05"),Weight=0.5))
    Weight=1
    ExecutePin=@n1.ExecutePin
   )
   > &RigUnit_ParentConstraint(
    AdvancedSettings=(InterpolationType=Average,RotationOrderForFilter=XZY)
    Child=(Type=Bone,Name="dyn_lw_tube_r_MSTR")
    Parents=((Item=(Type=Bone,Name="clavicle_r"),Weight=1),(Item=(Type=Bone,Name="spine_05"),Weight=1))
    Weight=1
   )
   > n2 = &RigUnit_SphericalPoseReader(
    ActiveRegionScaleFactors=(PositiveWidth=1,NegativeWidth=1,PositiveHeight=1,NegativeHeight=1)
    ActiveRegionSize=0.1
    Debug=(bDrawDebug=True,bDraw2D=False,bDrawLocalAxes=False,DebugScale=25,DebugSegments=20,DebugThickness=0.25)
    DriverAxis=(X=0,Y=1,Z=0)
    DriverItem=(Type=Bone,Name="head")
    FalloffRegionScaleFactors=(PositiveWidth=1,NegativeWidth=1,PositiveHeight=1,NegativeHeight=1)
    FalloffSize=0.3
    FlipHeightScaling=False
    FlipWidthScaling=False
    OptionalParentItem=(Type=Bone,Name="spine_05")
    RotationOffset=(X=160,Y=-15,Z=0)
   )
   > RigUnit_SetRelativeTransformForItem = /Script/ControlRig.RigUnit_SetRelativeTransformForItem(
    Child=(Type=Bone,Name="dyn_lw_tube_l_MSTR")
    Parent=(Type=Bone,Name="dyn_lw_tube_l_MSTR")
    Value=(Rotation=(X=0.1464,Y=-0.3536,Z=0.3536,W=0.8536),Translation=(X=0,Y=0,Z=0),Scale3D=(X=1,Y=1,Z=1))
    bParentInitial=False
    bPropagateToChildren=True
    Weight=@n2.OutputParam
   ) {
    is=struct RigVMUnitNode
    template="Set Relative Transform::Execute(in Child,in Parent,in bParentInitial,in Value,in Weight,in bPropagateToChildren)"
   }
   > n3 = &RigUnit_SphericalPoseReader(
    ActiveRegionScaleFactors=(PositiveWidth=1,NegativeWidth=1,PositiveHeight=1,NegativeHeight=1)
    ActiveRegionSize=0.1
    Debug=(bDrawDebug=true,bDraw2D=False,bDrawLocalAxes=False,DebugScale=25,DebugSegments=20,DebugThickness=0.25)
    DriverAxis=(X=0,Y=1,Z=0)
    DriverItem=(Type=Bone,Name="head")
    FalloffRegionScaleFactors=(PositiveWidth=1,NegativeWidth=1,PositiveHeight=1,NegativeHeight=1)
    FalloffSize=0.3
    FlipHeightScaling=False
    FlipWidthScaling=False
    OptionalParentItem=(Type=Bone,Name="spine_05")
    RotationOffset=(X=20,Y=15,Z=0)
   )
   RigVMFunction_MathBoolNot = &RigVMFunction_MathBoolNot(Value=RigVMVariableNode() {variable=isinFrontEnd}.Value)
   Cast = &RigVMFunction_MathBoolToFloat(Value=@RigVMFunction_MathBoolNot.Result)
   Multiply = &RigVMFunction_MathDoubleMul(B=@Cast.Result, A=@n3.OutputParam)
   > &RigUnit_SetRelativeTransformForItem(
    Child=(Type=Bone,Name="dyn_lw_tube_l_MSTR")
    Parent=(Type=Bone,Name="dyn_lw_tube_r_MSTR")
    Value=(Rotation=(X=-0.1464,Y=-0.3536,Z=-0.3536,W=0.8536),Translation=(X=0,Y=0,Z=0),Scale3D=(X=1,Y=1,Z=1))
    bParentInitial=False
    bPropagateToChildren=True
    Weight=@Multiply.Result
   ) {template="Set Relative Transform::Execute(in Child,in Parent,in bParentInitial,in Value,in Weight,in bPropagateToChildren)"}
   /Script/ControlRig.RigUnit_DebugTransformMutable(
    Color=(R=1,G=0,B=0,A=1)
    DebugDrawSettings=(DepthPriority=SDPG_Foreground,Lifetime=-1)
    ExecutePin=()
    Mode=Axes
    Scale=10
    Space=dyn_lw_tube_r_02
    Thickness=0
    Transform=(Rotation=(X=0,Y=0,Z=0,W=1),Translation=(X=0,Y=0,Z=0),Scale3D=(X=1,Y=1,Z=1))
    WorldOffset=(Rotation=(X=0,Y=0,Z=0,W=1),Translation=(X=0,Y=0,Z=0),Scale3D=(X=1,Y=1,Z=1))
    bEnabled=True
   ) {is=struct RigVMUnitNode}
   > &RigUnit_DebugTransformMutable(
    Color=(R=1,G=0,B=0,A=1)
    DebugDrawSettings=(DepthPriority=SDPG_Foreground,Lifetime=-1)
    Mode=Axes
    Scale=10
    Space=dyn_up_tube_r_MSTR
    Thickness=0
    Transform=(Rotation=(X=0,Y=0,Z=0,W=1),Translation=(X=0,Y=0,Z=0),Scale3D=(X=1,Y=1,Z=1))
    WorldOffset=(Rotation=(X=0,Y=0,Z=0,W=1),Translation=(X=0,Y=0,Z=0),Scale3D=(X=1,Y=1,Z=1))
    bEnabled=True
   )
   /Script/ControlRig.RigUnit_TransformConstraint(
    BaseBone=dyn_lw_tube_r_02
    BaseTransform=(Rotation=(X=0,Y=0,Z=0,W=1),Translation=(X=5,Y=0,Z=0),Scale3D=(X=1,Y=1,Z=1))
    BaseTransformSpace=GlobalSpace
    Bone=dyn_lw_tube_r_02
    ExecutePin=()
    Targets="((Transform=(Rotation=(X=0,Y=0,Z=0,W=1),Translation=(X=0,Y=0,Z=0),Scale3D=(X=1,Y=1,Z=1)),Weight=1,bMaintainOffset=True,Filter=(TranslationFilter=(bX=True,bY=True,bZ=True),RotationFilter=(bX=True,bY=True,bZ=True),ScaleFilter=(bX=True,bY=True,bZ=True))),(Tra ...(truncated)"
    bUseInitialTransforms=True
    Targets.0.Transform=&RigUnit_GetBoneTransform(Bone=dyn_up_tube_02_r).Transform
    Targets.1.Transform=&RigUnit_GetBoneTransform(Bone=dyn_lw_tube_r_01).Transform
   ) {is=struct RigVMUnitNode}
  ]}
]
hierarchy [
 root {loc=<0 0 0>, rot=<0 0 0 1>}
 attach {parent=root, loc=<0 0 0>, rot=<0 0 0 1>}
 pelvis {parent=root, loc=<0 -0.7071 93.1191>, rot=<0.0127 0.707 -0.0127 -0.707>}
 spine_01 {parent=pelvis, loc=<6.7494 -0.6388 0>, rot=<0 0 0.034 -0.9994>}
 spine_02 {parent=spine_01, loc=<7.2518 -0.035 0>, rot=<0 0 0.0238 -0.9997>}
 spine_03 {parent=spine_02, loc=<6.0104 0 0>, rot=<0 0 0.0149 -0.9999>}
 spine_04 {parent=spine_03, loc=<8.1024 0 0>, rot=<0 0 -0.0132 -0.9999>}
 spine_05 {parent=spine_04, loc=<10.0146 0 0>, rot=<0 0 -0.0015 -1>}
 clavicle_l {parent=spine_05, loc=<4.3575 -0.1466 -2.3223>, rot=<0.0465 -0.7289 -0.009 -0.6829>}
 dyn_lw_tube_l_MSTR {parent=clavicle_l, loc=<5.5402 -12.1895 2.8156>, rot=<-0.649 0.321 -0.2788 -0.6309>}
 dyn_lw_tube_l_DYN {parent=dyn_lw_tube_l_MSTR, loc=<0 0 0>, rot=<0 0 0 1>}
 dyn_lw_tube_l_01 {parent=dyn_lw_tube_l_DYN, loc=<0 0 0>, rot=<0 0 0 1>}
 dyn_lw_tube_l_02 {parent=dyn_lw_tube_l_01, loc=<4.1252 -0.2486 -0.1878>, rot=<-0.0275 -0.1356 -0.0448 -0.9894>}
 dyn_lw_tube_l_03 {parent=dyn_lw_tube_l_02, loc=<4.7243 -0.2335 -1.2823>, rot=<0 -0.3031 0 -0.953>}
 clavicle_r {parent=spine_05, loc=<4.3575 -0.1466 2.3223>, rot=<0.7289 0.0465 -0.6829 0.009>}
 dyn_lw_tube_r_MSTR {parent=clavicle_r, loc=<-5.5402 12.1895 -2.8152>, rot=<-0.2788 0.6309 0.649 0.321>}
 dyn_lw_tube_r_DYN {parent=dyn_lw_tube_r_MSTR, loc=<0 0 0>, rot=<0 0 0 1>}
 dyn_lw_tube_r_01 {parent=dyn_lw_tube_r_DYN, loc=<0 0 0>, rot=<0 0 0 1>}
 dyn_lw_tube_r_02 {parent=dyn_lw_tube_r_01, loc=<4.1253 0.2485 -0.1878>, rot=<0.0275 -0.1356 0.0448 -0.9894>}
 dyn_lw_tube_r_03 {parent=dyn_lw_tube_r_02, loc=<4.7247 0.2331 -1.2823>, rot=<0 -0.3031 0 -0.953>}
 neck_01 {parent=spine_05, loc=<11.3138 0.1529 0>, rot=<0 0 -0.2003 -0.9797>}
 neck_02 {parent=neck_01, loc=<4.5627 0 0>, rot=<0 0 0.0504 -0.9987>}
 head {parent=neck_02, loc=<5.3914 0 0>, rot=<0 0 0.115 -0.9934>}
 dyn_up_tube_l_MSTR {parent=head, loc=<-2.3357 4.6168 -6.7999>, rot=<0.1691 -0.7455 0.2435 -0.5969>}
 dyn_up_tube_l {parent=dyn_up_tube_l_MSTR, loc=<2.6357 0.8389 -0.5537>, rot=<0 0 0 1>}
 dyn_up_tube_l_DYN {parent=dyn_up_tube_l_MSTR, loc=<0 0 0>, rot=<0 0 0 1>}
 dyn_up_tube_01_l {parent=dyn_up_tube_l_DYN, loc=<0 0 0>, rot=<0 0 0 1>}
 dyn_up_tube_02_l {parent=dyn_up_tube_01_l, loc=<4.5884 0.5818 -0.5835>, rot=<-0.0043 -0.0118 0.1428 -0.9897>}
 dyn_up_tube_r_MSTR {parent=head, loc=<-2.3355 4.6168 6.7999>, rot=<0.5799 -0.2302 0.7696 -0.1361>}
 dyn_up_tube_r {parent=dyn_up_tube_r_MSTR, loc=<2.7152 -0.664 -0.3793>, rot=<-0.0011 0.0328 -0.0325 -0.9989>}
 dyn_up_tube_r_DYN {parent=dyn_up_tube_r_MSTR, loc=<0 0 0>, rot=<0 0 0 -1>}
 dyn_up_tube_01_r {parent=dyn_up_tube_r_DYN, loc=<0 0 0>, rot=<0 0 0 -1>}
 dyn_up_tube_02_r {parent=dyn_up_tube_01_r, loc=<4.6449 -0.2808 -0.2808>, rot=<0.0083 0.021 -0.1747 -0.9844>}
 dyn_up_tube_l_TGT {parent=spine_04, loc=<24.2085 -3.1328 -22.9524>, rot=<0.1969 -0.7387 0.221 -0.6056>}
 dyn_up_tube_r_TGT {parent=spine_04, loc=<24.2081 -3.1328 22.9524>, rot=<0.6056 -0.221 0.7387 -0.1969>}
]
variables [
 NewVar {type=bool, default=False}
 isinFrontEnd {type=bool, default=False}
]
