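# Physics asset for the medium male Mastermind head cosmetic, set up for secondary motion on four tubes (upper and lower, left and right).
# Each tube has a DYN capsule constrained to its MSTR capsule with narrow, angle-limited swing and twist.
# Upper MSTR bodies link to the head and lower ones to their side's clavicle through angularly free constraints.
# Larger capsules represent the head and both clavicles.
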PhysicsAsset /BRCosmetics/Characters/Player/Male/Medium/Heads/M_MED_Mastermind_Head/Meshes/Male_Medium_Mastermind_Head_Dyn
bodies {
 clavicle_l [{shapeType=Capsule, length=8.1244, radius=9.0086, center=<7.5 0 0>, rotation=<rot 90 0 0>}]
 clavicle_r [{shapeType=Capsule, length=8.1244, radius=9.0086, center=<-7.5 0 0>, rotation=<rot 90 0 0>}]
 dyn_lw_tube_l_DYN [{shapeType=Capsule, length=4, radius=2, center=<1 0 0>, rotation=<rot 90 0 0>}]
 dyn_lw_tube_l_MSTR [{shapeType=Capsule, length=1, radius=1}]
 dyn_lw_tube_r_DYN [{shapeType=Capsule, length=4, radius=2, center=<1 0 0>, rotation=<rot 90 0 0>}]
 dyn_lw_tube_r_MSTR [{shapeType=Capsule, length=1, radius=1}]
 dyn_up_tube_l_DYN [{shapeType=Capsule, length=4, radius=2, center=<2 0 0>, rotation=<rot 90 0 0>}]
 dyn_up_tube_l_MSTR [{shapeType=Capsule, length=1, radius=1}]
 dyn_up_tube_r_DYN [{shapeType=Capsule, length=4, radius=2, center=<2 0 0>, rotation=<rot 90 0 0>}]
 dyn_up_tube_r_MSTR [{shapeType=Capsule, length=1, radius=1}]
 head [{shapeType=Capsule, length=6.0563, radius=11.4604, center=<3.9797 0 0.0001>, rotation=<rot 90 0 0>}]
}
constraints [
 {bone1Name=dyn_lw_tube_r_DYN, bone2Name=dyn_lw_tube_r_MSTR, swing1LimitDegrees=10, swing2LimitDegrees=30, twistLimitDegrees=5}
 {bone1Name=dyn_lw_tube_l_DYN, bone2Name=dyn_lw_tube_l_MSTR, swing1LimitDegrees=10, swing2LimitDegrees=30, twistLimitDegrees=5}
 {bone1Name=dyn_up_tube_r_DYN, bone2Name=dyn_up_tube_r_MSTR, swing1LimitDegrees=5, swing2LimitDegrees=30, twistLimitDegrees=5}
 {bone1Name=dyn_up_tube_l_DYN, bone2Name=dyn_up_tube_l_MSTR, swing1LimitDegrees=5, swing2LimitDegrees=30, twistLimitDegrees=5}
 {
  bone1Name=dyn_up_tube_l_MSTR
  bone2Name=head
  swing1LimitDegrees=45
  swing1Motion=Free
  swing2LimitDegrees=45
  swing2Motion=Free
  twistLimitDegrees=45
  twistMotion=Free
 }
 {
  bone1Name=dyn_up_tube_r_MSTR
  bone2Name=head
  swing1LimitDegrees=45
  swing1Motion=Free
  swing2LimitDegrees=45
  swing2Motion=Free
  twistLimitDegrees=45
  twistMotion=Free
 }
 {
  bone1Name=dyn_lw_tube_l_MSTR
  bone2Name=clavicle_l
  swing1LimitDegrees=45
  swing1Motion=Free
  swing2LimitDegrees=45
  swing2Motion=Free
  twistLimitDegrees=45
  twistMotion=Free
 }
 {
  bone1Name=dyn_lw_tube_r_MSTR
  bone2Name=clavicle_r
  swing1LimitDegrees=45
  swing1Motion=Free
  swing2LimitDegrees=45
  swing2Motion=Free
  twistLimitDegrees=45
  twistMotion=Free
 }
]
