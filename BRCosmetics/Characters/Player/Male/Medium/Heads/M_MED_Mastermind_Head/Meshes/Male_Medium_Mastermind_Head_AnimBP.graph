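# Anim Blueprint for the male Mastermind cosmetic head that reshapes its tube bones to follow head, shoulder and aiming movement.
# Each update maps head and clavicle rotations relative to spine_05, plus the targeting flag, to corrective weights (head_down is computed but unused).
# The copied parent pose passes through chained, mostly additive Modify Bone nodes, a Control Rig given the front-end flag, then RigidBody simulation in spine_05 space.
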AnimBlueprint /BRCosmetics/Characters/Player/Male/Medium/Heads/M_MED_Mastermind_Head/Meshes/Male_Medium_Mastermind_Head_AnimBP
target_skeleton=&/Male_Medium_Mastermind_Head_Skeleton
graphs [
 AnimGraph {nodes [
   n1 = Get head_l() -> (head_l:double)
   n2 = "Transform (Modify) Bone - Bone: dyn_lw_tube_l_MSTR"(
    ComponentPose=Local To Component(LocalPose=Copy Pose From Mesh() {bUseAttachedParent=true})
    Alpha=@n1
    Translation="0,0,0"
    Rotation="40,-20,0.000"
    Scale="1,1,1"
   ) {
    rotationMode=BMM_Additive
    rotationSpace=BCS_BoneSpace
    scaleMode=BMM_Additive
    scaleSpace=BCS_BoneSpace
    translationMode=BMM_Additive
    translationSpace=BCS_BoneSpace
    alphaBoolBlend {blendOption=Linear}
    boneToModify {boneName=dyn_lw_tube_l_MSTR}
    rotation {pitch=40, yaw=-20}
   }
   n3 = "Transform (Modify) Bone - Bone: dyn_up_tube_02_l"(ComponentPose=@n2, Alpha=@n1, Translation="0,0,0", Rotation="0.000,20,0.000", Scale="0.8,1,1") {
    rotationMode=BMM_Additive
    rotationSpace=BCS_BoneSpace
    scaleMode=BMM_Additive
    scaleSpace=BCS_BoneSpace
    translationMode=BMM_Additive
    translationSpace=BCS_BoneSpace
    alphaBoolBlend {blendOption=Linear}
    boneToModify {boneName=dyn_up_tube_02_l}
    rotation {yaw=20}
    scale {x=0.8}
   }
   n4 = "Transform (Modify) Bone - Bone: dyn_lw_tube_l_01"(ComponentPose=@n3, Alpha=@n1, Translation="-1,0,0", Rotation="0,0,0", Scale="0.9,1,1") {
    rotationMode=BMM_Additive
    rotationSpace=BCS_BoneSpace
    scaleMode=BMM_Additive
    scaleSpace=BCS_BoneSpace
    translationMode=BMM_Additive
    translationSpace=BCS_BoneSpace
    alphaBoolBlend {blendOption=Linear}
    boneToModify {boneName=dyn_lw_tube_l_01}
    scale {x=0.9}
    translation {x=-1}
   }
   n5 = "Transform (Modify) Bone - Bone: dyn_lw_tube_l_02"(ComponentPose=@n4, Alpha=@n1, Translation="-1,0,0", Rotation="0,0,0", Scale="0.9,1,1") {
    rotationMode=BMM_Additive
    rotationSpace=BCS_BoneSpace
    scaleMode=BMM_Additive
    scaleSpace=BCS_BoneSpace
    translationMode=BMM_Additive
    translationSpace=BCS_BoneSpace
    alphaBoolBlend {blendOption=Linear}
    boneToModify {boneName=dyn_lw_tube_l_02}
    scale {x=0.9}
    translation {x=-1}
   }
   n6 = "Transform (Modify) Bone - Bone: dyn_lw_tube_l_03"(ComponentPose=@n5, Alpha=@n1, Translation="-1,0,0", Rotation="0.000,-15.000,0.000", Scale="0.9,1,1") {
    rotationMode=BMM_Additive
    rotationSpace=BCS_BoneSpace
    scaleMode=BMM_Additive
    scaleSpace=BCS_BoneSpace
    translationMode=BMM_Additive
    translationSpace=BCS_BoneSpace
    alphaBoolBlend {blendOption=Linear}
    boneToModify {boneName=dyn_lw_tube_l_03}
    rotation {yaw=-15.0001}
    scale {x=0.9}
    translation {x=-1}
   }
   n7 = "Transform (Modify) Bone - Bone: dyn_lw_tube_r_MSTR"(ComponentPose=@n6, Alpha=@n1, Translation="0,0,0", Rotation="-15,-20,0.000", Scale="1,1,1") {
    rotationMode=BMM_Additive
    rotationSpace=BCS_BoneSpace
    scaleMode=BMM_Additive
    scaleSpace=BCS_BoneSpace
    translationMode=BMM_Additive
    translationSpace=BCS_BoneSpace
    alphaBoolBlend {blendOption=Linear}
    boneToModify {boneName=dyn_lw_tube_r_MSTR}
    rotation {pitch=-15, yaw=-20}
   }
   n8 = "Transform (Modify) Bone - Bone: dyn_lw_tube_r_03"(ComponentPose=@n7, Alpha=1, Translation="1,0,0", Rotation="0.000,-15.000,0.000", Scale="1,1,1") {
    rotationMode=BMM_Additive
    rotationSpace=BCS_BoneSpace
    scaleMode=BMM_Additive
    scaleSpace=BCS_BoneSpace
    translationMode=BMM_Additive
    translationSpace=BCS_BoneSpace
    alphaBoolBlend {blendOption=Linear}
    boneToModify {boneName=dyn_lw_tube_r_03}
    rotation {yaw=-15.0001}
    translation {x=1}
   }
   n9 = Get head_r() -> (head_r:double)
   n10 = "Transform (Modify) Bone - Bone: dyn_lw_tube_r_MSTR"(ComponentPose=@n8, Alpha=@n9, Translation="0,0,0", Rotation="15,20,0.000", Scale="1,1,1") {
    rotationMode=BMM_Additive
    rotationSpace=BCS_BoneSpace
    scaleMode=BMM_Additive
    scaleSpace=BCS_BoneSpace
    translationMode=BMM_Additive
    translationSpace=BCS_BoneSpace
    alphaBoolBlend {blendOption=Linear}
    boneToModify {boneName=dyn_lw_tube_r_MSTR}
    rotation {pitch=15, yaw=20}
   }
   n11 = "Transform (Modify) Bone - Bone: dyn_up_tube_02_r"(ComponentPose=@n10, Alpha=@n9, Translation="0,0,0", Rotation="0.000,-20,0.000", Scale="0.8,1,1") {
    rotationMode=BMM_Additive
    rotationSpace=BCS_BoneSpace
    scaleMode=BMM_Additive
    scaleSpace=BCS_BoneSpace
    translationMode=BMM_Additive
    translationSpace=BCS_BoneSpace
    alphaBoolBlend {blendOption=Linear}
    boneToModify {boneName=dyn_up_tube_02_r}
    rotation {yaw=-20}
    scale {x=0.8}
   }
   n12 = "Transform (Modify) Bone - Bone: dyn_lw_tube_r_01"(ComponentPose=@n11, Alpha=@n9, Translation="-0.5,0,0", Rotation="0,0,0", Scale="0.9,1,1") {
    rotationMode=BMM_Additive
    rotationSpace=BCS_BoneSpace
    scaleMode=BMM_Additive
    scaleSpace=BCS_BoneSpace
    translationMode=BMM_Additive
    translationSpace=BCS_BoneSpace
    alphaBoolBlend {blendOption=Linear}
    boneToModify {boneName=dyn_lw_tube_r_01}
    scale {x=0.9}
    translation {x=-0.5}
   }
   n13 = "Transform (Modify) Bone - Bone: dyn_lw_tube_r_02"(ComponentPose=@n12, Alpha=@n9, Translation="-0.5,0,0", Rotation="0,0,0", Scale="0.9,1,1") {
    rotationMode=BMM_Additive
    rotationSpace=BCS_BoneSpace
    scaleMode=BMM_Additive
    scaleSpace=BCS_BoneSpace
    translationMode=BMM_Additive
    translationSpace=BCS_BoneSpace
    alphaBoolBlend {blendOption=Linear}
    boneToModify {boneName=dyn_lw_tube_r_02}
    scale {x=0.9}
    translation {x=-0.5}
   }
   n14 = "Transform (Modify) Bone - Bone: dyn_lw_tube_r_03"(ComponentPose=@n13, Alpha=@n9, Translation="-0.5,0,0", Rotation="0.000,-15.000,0.000", Scale="1,1,1") {
    rotationMode=BMM_Additive
    rotationSpace=BCS_BoneSpace
    scaleMode=BMM_Additive
    scaleSpace=BCS_BoneSpace
    translationMode=BMM_Additive
    translationSpace=BCS_BoneSpace
    alphaBoolBlend {blendOption=Linear}
    boneToModify {boneName=dyn_lw_tube_r_03}
    rotation {yaw=-15.0001}
    translation {x=-0.5}
   }
   n15 = Get head_r_more() -> (head_r_more:double)
   n16 = "Transform (Modify) Bone - Bone: dyn_up_tube_01_r"(ComponentPose=@n14, Alpha=@n15, Translation="0,0,0", Rotation="0,0,0", Scale="0.6,1,1") {
    rotationMode=BMM_Additive
    rotationSpace=BCS_BoneSpace
    scaleMode=BMM_Additive
    scaleSpace=BCS_BoneSpace
    translationMode=BMM_Additive
    translationSpace=BCS_BoneSpace
    alphaBoolBlend {blendOption=Linear}
    boneToModify {boneName=dyn_up_tube_01_r}
    scale {x=0.6}
   }
   n17 = "Transform (Modify) Bone - Bone: dyn_up_tube_02_r"(ComponentPose=@n16, Alpha=@n15, Translation="0,0,0", Rotation="15,0,0.000", Scale="0.5,1,1") {
    rotationMode=BMM_Additive
    rotationSpace=BCS_BoneSpace
    scaleMode=BMM_Additive
    scaleSpace=BCS_BoneSpace
    translationMode=BMM_Additive
    translationSpace=BCS_BoneSpace
    alphaBoolBlend {blendOption=Linear}
    boneToModify {boneName=dyn_up_tube_02_r}
    rotation {pitch=15}
    scale {x=0.5}
   }
   n18 = "Transform (Modify) Bone - Bone: dyn_lw_tube_r_MSTR"(ComponentPose=@n17, Alpha=@n15, Translation="0,0,0", Rotation="0,15,0", Scale="1,1,1") {
    rotationMode=BMM_Additive
    rotationSpace=BCS_BoneSpace
    scaleMode=BMM_Additive
    scaleSpace=BCS_BoneSpace
    translationMode=BMM_Additive
    translationSpace=BCS_BoneSpace
    alphaBoolBlend {blendOption=Linear}
    boneToModify {boneName=dyn_lw_tube_r_MSTR}
    rotation {yaw=15}
   }
   n19 = "Transform (Modify) Bone - Bone: dyn_up_tube_r_TGT"(ComponentPose=@n18, Alpha=@n15, Translation="-7.2,0,5.8", Rotation="0,0,0", Scale="1,1,1") {
    translationMode=BMM_Additive
    translationSpace=BCS_BoneSpace
    alphaBoolBlend {blendOption=Linear}
    boneToModify {boneName=dyn_up_tube_r_TGT}
    translation {x=-7.2, z=5.8}
   }
   n20 = "Transform (Modify) Bone - Bone: dyn_up_tube_01_l"(ComponentPose=@n19, Alpha=@n15, Translation="3,0,0", Rotation="0,0,0", Scale="1,1,1") {
    rotationMode=BMM_Additive
    rotationSpace=BCS_BoneSpace
    scaleMode=BMM_Additive
    scaleSpace=BCS_BoneSpace
    translationMode=BMM_Additive
    translationSpace=BCS_BoneSpace
    alphaBoolBlend {blendOption=Linear}
    boneToModify {boneName=dyn_up_tube_01_l}
    translation {x=3}
   }
   n21 = "Transform (Modify) Bone - Bone: dyn_up_tube_02_l"(ComponentPose=@n20, Alpha=@n15, Translation="3,0,0", Rotation="0,0,0", Scale="1,1,1") {
    rotationMode=BMM_Additive
    rotationSpace=BCS_BoneSpace
    scaleMode=BMM_Additive
    scaleSpace=BCS_BoneSpace
    translationMode=BMM_Additive
    translationSpace=BCS_BoneSpace
    alphaBoolBlend {blendOption=Linear}
    boneToModify {boneName=dyn_up_tube_02_l}
    translation {x=3}
   }
   n22 = "Transform (Modify) Bone - Bone: dyn_lw_tube_l_MSTR"(ComponentPose=@n21, Alpha=@n15, Translation="0,0,0", Rotation="-15,0.000,0.000", Scale="1,1,1") {
    rotationMode=BMM_Additive
    rotationSpace=BCS_BoneSpace
    scaleMode=BMM_Additive
    scaleSpace=BCS_BoneSpace
    translationMode=BMM_Additive
    translationSpace=BCS_BoneSpace
    alphaBoolBlend {blendOption=Linear}
    boneToModify {boneName=dyn_lw_tube_l_MSTR}
    rotation {pitch=-15}
   }
   n23 = "Transform (Modify) Bone - Bone: dyn_lw_tube_l_01"(ComponentPose=@n22, Alpha=@n15, Translation="3,0,0", Rotation="-15,0,0", Scale="1,1,1") {
    rotationMode=BMM_Additive
    rotationSpace=BCS_BoneSpace
    scaleMode=BMM_Additive
    scaleSpace=BCS_BoneSpace
    translationMode=BMM_Additive
    translationSpace=BCS_BoneSpace
    alphaBoolBlend {blendOption=Linear}
    boneToModify {boneName=dyn_lw_tube_l_01}
    rotation {pitch=-15}
    translation {x=3}
   }
   n24 = "Transform (Modify) Bone - Bone: dyn_lw_tube_l_02"(ComponentPose=@n23, Alpha=@n15, Translation="3,0,0", Rotation="-15,0,0", Scale="1,1,1") {
    rotationMode=BMM_Additive
    rotationSpace=BCS_BoneSpace
    scaleMode=BMM_Additive
    scaleSpace=BCS_BoneSpace
    translationMode=BMM_Additive
    translationSpace=BCS_BoneSpace
    alphaBoolBlend {blendOption=Linear}
    boneToModify {boneName=dyn_lw_tube_l_02}
    rotation {pitch=-15}
    translation {x=3}
   }
   n25 = "Transform (Modify) Bone - Bone: dyn_lw_tube_l_03"(ComponentPose=@n24, Alpha=@n15, Translation="3,0,0", Rotation="-15,0,0", Scale="1,1,1") {
    rotationMode=BMM_Additive
    rotationSpace=BCS_BoneSpace
    scaleMode=BMM_Additive
    scaleSpace=BCS_BoneSpace
    translationMode=BMM_Additive
    translationSpace=BCS_BoneSpace
    alphaBoolBlend {blendOption=Linear}
    boneToModify {boneName=dyn_lw_tube_l_03}
    rotation {pitch=-15}
    translation {x=3}
   }
   n26 = "Transform (Modify) Bone - Bone: dyn_lw_tube_r_MSTR"(ComponentPose=@n25, Alpha=Get right_arm_up() -> (right_arm_up:double), Translation="0,0,0", Rotation="10,60,0.000", Scale="0.9,1,1") {
    rotationMode=BMM_Additive
    rotationSpace=BCS_BoneSpace
    translationMode=BMM_Additive
    translationSpace=BCS_BoneSpace
    alphaBoolBlend {blendOption=Linear}
    alphaScaleBiasClamp {bInterpResult=true, interpSpeedDecreasing=30, interpSpeedIncreasing=30}
    boneToModify {boneName=dyn_lw_tube_r_MSTR}
    rotation {pitch=10, yaw=60}
    scale {x=0.9}
   }
   n27 = "Transform (Modify) Bone - Bone: dyn_lw_tube_l_MSTR"(ComponentPose=@n26, Alpha=Get left_arm_up() -> (left_arm_up:double), Translation="0,0,0", Rotation="10,-60,0.000", Scale="0.9,1,1") {
    rotationMode=BMM_Additive
    rotationSpace=BCS_BoneSpace
    translationMode=BMM_Additive
    translationSpace=BCS_BoneSpace
    alphaBoolBlend {blendOption=Linear}
    alphaScaleBiasClamp {bInterpResult=true, interpSpeedDecreasing=30, interpSpeedIncreasing=30}
    boneToModify {boneName=dyn_lw_tube_l_MSTR}
    rotation {pitch=10, yaw=-60}
    scale {x=0.9}
   }
   n28 = Get is_ADS() -> (is_ADS:double)
   n29 = "Transform (Modify) Bone - Bone: dyn_lw_tube_r_01"(ComponentPose=@n27, Alpha=@n28, Translation="0,0,0", Rotation="0,0,0", Scale="0.9,1,1") {
    rotationMode=BMM_Additive
    rotationSpace=BCS_BoneSpace
    scaleMode=BMM_Additive
    scaleSpace=BCS_BoneSpace
    translationMode=BMM_Additive
    translationSpace=BCS_BoneSpace
    alphaBoolBlend {blendOption=Linear}
    alphaScaleBiasClamp {bInterpResult=true, interpSpeedDecreasing=30, interpSpeedIncreasing=30}
    boneToModify {boneName=dyn_lw_tube_r_01}
    scale {x=0.9}
   }
   n30 = "Transform (Modify) Bone - Bone: dyn_lw_tube_r_02"(ComponentPose=@n29, Alpha=@n28, Translation="0,0,0", Rotation="0,0,0", Scale="0.9,1,1") {
    rotationMode=BMM_Additive
    rotationSpace=BCS_BoneSpace
    scaleMode=BMM_Additive
    scaleSpace=BCS_BoneSpace
    translationMode=BMM_Additive
    translationSpace=BCS_BoneSpace
    alphaBoolBlend {blendOption=Linear}
    alphaScaleBiasClamp {bInterpResult=true, interpSpeedDecreasing=30, interpSpeedIncreasing=30}
    boneToModify {boneName=dyn_lw_tube_r_02}
    scale {x=0.9}
   }
   n31 = "Transform (Modify) Bone - Bone: dyn_lw_tube_r_03"(ComponentPose=@n30, Alpha=@n28, Translation="0,0,0", Rotation="0,0,0", Scale="0.9,1,1") {
    rotationMode=BMM_Additive
    rotationSpace=BCS_BoneSpace
    scaleMode=BMM_Additive
    scaleSpace=BCS_BoneSpace
    translationMode=BMM_Additive
    translationSpace=BCS_BoneSpace
    alphaBoolBlend {blendOption=Linear}
    alphaScaleBiasClamp {bInterpResult=true, interpSpeedDecreasing=30, interpSpeedIncreasing=30}
    boneToModify {boneName=dyn_lw_tube_r_03}
    scale {x=0.9}
   }
   n32 = "Transform (Modify) Bone - Bone: dyn_up_tube_01_r"(ComponentPose=@n31, Alpha=@n28, Translation="0,0,0", Rotation="0,0,0", Scale="0.9,1,1") {
    rotationMode=BMM_Additive
    rotationSpace=BCS_BoneSpace
    scaleMode=BMM_Additive
    scaleSpace=BCS_BoneSpace
    translationMode=BMM_Additive
    translationSpace=BCS_BoneSpace
    alphaBoolBlend {blendOption=Linear}
    alphaScaleBiasClamp {bInterpResult=true, interpSpeedDecreasing=30, interpSpeedIncreasing=30}
    boneToModify {boneName=dyn_up_tube_01_r}
    scale {x=0.9}
   }
   n33 = "Transform (Modify) Bone - Bone: dyn_up_tube_02_r"(ComponentPose=@n32, Alpha=@n28, Translation="0,0,0", Rotation="30,0,0", Scale="0.9,1,1") {
    rotationMode=BMM_Additive
    rotationSpace=BCS_BoneSpace
    scaleMode=BMM_Additive
    scaleSpace=BCS_BoneSpace
    translationMode=BMM_Additive
    translationSpace=BCS_BoneSpace
    alphaBoolBlend {blendOption=Linear}
    alphaScaleBiasClamp {bInterpResult=true, interpSpeedDecreasing=30, interpSpeedIncreasing=30}
    boneToModify {boneName=dyn_up_tube_02_r}
    rotation {pitch=30}
    scale {x=0.9}
   }
   n34 = Control Rig!(Source=Component To Local(ComponentPose=@n33), Alpha, isinFrontEnd:bool=Get bIsInFrontEnd()) -> (Pose) {
    alphaBoolBlend {blendOption=Linear}
    controlRigAssetReference {blueprintRigClass=&/M_MED_Mastermind_CtrlRig.M_MED_Mastermind_CtrlRig_C}
   }
   n35 = RigidBody(ComponentPose=Local To Component(LocalPose=@n34)) {
    bClampLinearTranslationLimitToRefPose=true
    bOverrideWorldGravity=true
    overridePhysicsAsset=&/Male_Medium_Mastermind_Head_Dyn
    simulationSpace=BaseBoneSpace
    alphaBoolBlend {blendOption=Linear}
    baseBoneRef {boneName=spine_05}
    componentAppliedLinearAccClamp=<3000 3000 3000>
    componentLinearAccScale=<0.2 0.3 0.3>
    componentLinearVelScale=<0.2 0.3 0.3>
    overrideWorldGravity {z=-980}
   }
   Output Pose(Result=Component To Local(ComponentPose=@n35)) {layerGroup=None}
  ]}
 EventGraph {nodes [
   Event BlueprintUpdateAnimation()
   > n1 = Sequence!(execute) -> (then_0, then_1, then_2:Exec)
   n2 = GetOwnerSkeletalMeshComponent()
   n3 = BreakRotator(InRot=BreakTransform(InTransform=GetDeltaTransformFromRefPose(self=@n2, BoneName=head, BaseName=spine_05)).Rotation)
   > Set head_l(head_l:double=MapRangeClamped(Value=@n3, InRangeA=0, InRangeB=60)) -> (Output_Get:double)
   > Set head_r(head_r:double=MapRangeClamped(Value=@n3, InRangeA=0, InRangeB=-40)) -> (Output_Get:double)
   > Set head_r_more(head_r_more:double=MapRangeClamped(Value=@n3, InRangeA=-25, InRangeB=-70)) -> (Output_Get:double)
   > Set head_down(head_down:double=MapRangeClamped(Value=@n3.Yaw, InRangeA=0, InRangeB=20)) -> (Output_Get:double)
   Set is_ADS(execute=@n1.then_2, is_ADS:double="To Float (Boolean)"(InBool=Get bIsTargeting())) -> (Output_Get:double)
   n4 = BreakTransform(InTransform=GetDeltaTransformFromRefPose(self=GetOwnerSkeletalMeshComponent(), BoneName=clavicle_r, BaseName=spine_05))
   Set right_arm_up(execute=@n1.then_1, right_arm_up:double=MapRangeClamped(Value=BreakRotator(InRot=@n4.Rotation).Pitch, InRangeA=10, InRangeB=40)) -> (Output_Get:double)
   n5 = BreakTransform(InTransform=GetDeltaTransformFromRefPose(self=GetOwnerSkeletalMeshComponent(), BoneName=clavicle_l, BaseName=spine_05))
   > Set left_arm_up(left_arm_up:double=MapRangeClamped(Value=BreakRotator(InRot=@n5.Rotation).Pitch, InRangeA=10, InRangeB=40)) -> (Output_Get:double)
  ]}
]
variables [
 head_down {type=number}
 head_l {type=number}
 head_r {type=number}
 head_r_more {type=number}
 is_ADS {type=number}
 left_arm_up {type=number}
 right_arm_up {type=number}
]
